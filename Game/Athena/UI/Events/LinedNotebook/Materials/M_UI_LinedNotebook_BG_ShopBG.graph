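# Translucent UI material that draws an animated, smoky shop background for the LinedNotebook event.
# A radial gradient samples curve-atlas colors, with a GreenColorScheme static switch choosing green or purple for gradient, smoke tint and spotlight.
# Panning cloud textures form radially masked background smoke, and panning puff textures form gradient-masked foreground smoke.
# Named reroutes pass screen-scaled UVs and intermediate layers between sections, including a circle spotlight with offset-perturbed UVs.
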Material /Game/Athena/UI/Events/LinedNotebook/Materials/M_UI_LinedNotebook_BG_ShopBG
graph {
 properties {
  bUsedWithStaticMesh=false
  blendMode=BLEND_Translucent
  materialDomain=MD_UI
  pixelDepthOffsetMode=PDOM_Legacy
  usedShadingModels=Default Lit
 }
 nodes [
  n1 = NamedRerouteUsage()
  n2 = ComponentMask(None=MaterialFunctionCall!(Dimensions) {materialFunction=/FortUILibrary/MaterialFunctions/Basics/MF_UI_GetAspectRatio}.H Ratio) {r=true}
  n3 = AppendVector(A=Multiply(A=ComponentMask(None=@n1) {r=true}, B=@n2) {constB=0.73}, B=ComponentMask(None=@n1) {g=true})
  n4 = AppendVector(A=Divide(A=OneMinus(None=@n2)), B=Constant())
  Add(A=Multiply(A=Add(A=@n3, B=@n4), B=Constant2Vector() {g=1, r=0.55}), B=Constant2Vector() {r=0.225})
  n5 = VectorParameter() {group=Fg_Smoke, parameterName=TransparentSmoke_ColorLight, defaultValue=<rgba 0.15 0.0103 0.2159 1>}
  n6 = VectorParameter() {group=Fg_Smoke, parameterName=TransparentSmoke_ColorDark, defaultValue=<rgba 0.0137 0.0015 0.0685 1>}
  n7 = NamedRerouteUsage()
  n8 = Constant()
  n9 = Constant() {r=0.75}
  n10 = Constant() {r=0.25}
  n11 = Constant() {r=-0.15}
  n12 = NamedRerouteUsage()
  n13 = NamedRerouteUsage()
  n14 = ScalarParameter() {defaultValue=1, parameterName=UseFullSmokeBackground, sliderMax=1}
  LinearInterpolate(
   A=LinearInterpolate(
    A=MaterialFunctionCall(R=@n5.R, G=@n5.R, B=@n5.R) {materialFunction=&AppendMany}.RGBA
    B=MaterialFunctionCall(R=@n6.R, G=@n6.R, B=@n6.R) {materialFunction=&AppendMany}.RGBA
    Alpha=MaterialFunctionCall(UVs=@n7, GradientStart=@n8, GradientEnd=@n9, RotationAngle=@n10, GradientOffset=@n11) {materialFunction=&MF_UI_LinearGradient}.Result
   )
   B=Saturate(None=Add(A=@n12, B=@n13))
   Alpha=@n14
  )
  LinearInterpolate(A=NamedRerouteUsage(), Alpha=@n14)
  n15 = AppendVector(A=Constant() {r=1}, B=Constant() {r=0.9})
  Multiply(A=@n15) {constB=1.6}
  n16 = AppendVector(A=Constant() {r=1.4}, B=Constant() {r=1.32})
  Multiply(A=@n16) {constB=1.6}
  n17 = VectorParameter() {group=Spotlight, parameterName=Spotlight_Green, defaultValue=<rgba 0.159 0.6724 0.0437 1>}
  n18 = VectorParameter() {group=Fg_Smoke, parameterName=Spotlight_Purple, defaultValue=<rgba 0.1318 0 0.7031 1>}
  n19 = StaticSwitchParameter(
   True=MaterialFunctionCall(R=@n17.R, G=@n17.R, B=@n17.R) {materialFunction=&AppendMany}.RGBA
   False=MaterialFunctionCall(R=@n18.R, G=@n18.R, B=@n18.R) {materialFunction=&AppendMany}.RGBA
  ) {defaultValue=true, parameterName=GreenColorScheme}
  n20 = NamedRerouteUsage()
  n21 = Add(A=@n20, B=NamedRerouteUsage())
  n22 = NamedRerouteDeclaration(None=MaterialFunctionCall(
    UVs=LinearInterpolate(A=@n20, B=@n21) {constAlpha=0.1}
    Size=ScalarParameter() {defaultValue=0.1, group=Spotlight, parameterName=Spotlight_Size}
    GlowMin=ScalarParameter() {defaultValue=0.1, group=Spotlight, parameterName=Spotlight_GlowMin}
    GlowMax=ScalarParameter() {defaultValue=0.8, group=Spotlight, parameterName=Spotlight_GlowMax}
   ) {materialFunction=&MF_UI_Circle}.Fill) {name=SpotlightCircleShape, nodeColor=<rgba 0 0.2384 1 1>}
  n23 = ScalarParameter() {defaultValue=0.2, group=Spotlight, parameterName=GreenSpotlightMultiplier}
  n24 = ScalarParameter() {defaultValue=0.2, group=Spotlight, parameterName=PurpleSpotlightMultiplier}
  NamedRerouteDeclaration(None=Multiply(A=LinearInterpolate(B=@n19, Alpha=@n22), B=StaticSwitchParameter(True=@n23, False=@n24) {defaultValue=true, parameterName=GreenColorScheme}) {constB=0.2}) {name=Spotlight, nodeColor=<rgba 0 0.2384 1 1>}
  NamedRerouteDeclaration(None=NamedRerouteUsage()) {name=Radial Gradient UV, nodeColor=<rgba 0 0 0 1>}
  n25 = LinearInterpolate(B=NamedRerouteUsage(), Alpha=OneMinus(None=NamedRerouteUsage()))
  n26 = MaterialFunctionCall(
   UVs=NamedRerouteUsage()
   GradientStart=ScalarParameter() {group=Fg_Smoke, parameterName=FG_Smoke_GradMaskStart}
   GradientEnd=ScalarParameter() {defaultValue=1, group=Fg_Smoke, parameterName=FG_Smoke_GradMaskEnd}
   RotationAngle=Constant() {r=0.25}
  ) {materialFunction=&MF_UI_LinearGradient}
  NamedRerouteDeclaration(None=LinearInterpolate(A=@n25, B=Multiply(A=@n25, B=@n26.Result), Alpha=ScalarParameter() {group=Fg_Smoke, parameterName=FG_Smoke_LinearGradMask, sliderMax=1})) {name=MaskedForegroundSmoke, nodeColor=<rgba 0 0.2384 1 1>}
  n27 = NamedRerouteUsage()
  n28 = LinearInterpolate(A=@n27, B=Add(A=@n27, B=NamedRerouteUsage())) {constAlpha=0.05}
  n29 = NamedRerouteUsage()
  NamedRerouteDeclaration(None=TextureSample(UVs=Panner(Coordinate=@n28, Time=@n29) {speedX=-0.05}) {samplerType=SAMPLERTYPE_LinearColor, texture=/Game/Athena/UI/Events/LinedNotebook/Textures/Smoke512}.RGB) {name=Fg_Smoketexture, nodeColor=<rgba 0 0.2384 1 1>}
  n30 = MaterialFunctionCall!(ScreenUV, TextureSize, ScalePivotPoint) {materialFunction=/FortUILibrary/MaterialFunctions/Basics/MF_UI_ScreenSpaceUVScaling}
  NamedRerouteDeclaration(None=@n30.Scaled to Fill) {name=Normal UV Scaled to Fill, nodeColor=<rgba 0 0 0 1>}
  n31 = ScalarParameter() {defaultValue=90, group=Fg_Smoke, parameterName=Fg_Smoke_Angle}
  n32 = Constant()
  n33 = Constant() {r=-360}
  n34 = Constant() {r=1}
  NamedRerouteDeclaration(None=MaterialFunctionCall(
    UVs=NamedRerouteUsage()
    GradientStart=ScalarParameter() {defaultValue=-0.3, group=Fg_Smoke, parameterName=Fg_Smoke_Endpoint}
    GradientEnd=ScalarParameter() {defaultValue=1.6, group=Fg_Smoke, parameterName=Fg_Smoke_Intensity}
    RotationAngle=MaterialFunctionCall(Input=@n31, Input Low=@n32, Input High=@n33, Target Low=@n32, Target High=@n34) {materialFunction=&RemapValueRange}.Result
   ) {materialFunction=&MF_UI_LinearGradient}.Result) {name=Smoke_LinearMask, nodeColor=<rgba 0 0.2384 1 1>}
  n35 = AppendVector(
   A=ScalarParameter() {defaultValue=0.4, group=Fg_Smoke, parameterName=Fg_Smoke_U1}
   B=ScalarParameter() {defaultValue=0.4, group=Fg_Smoke, parameterName=Fg_Smoke_V1}
  )
  n36 = NamedRerouteUsage()
  n37 = ScalarParameter() {defaultValue=1, group=Fg_Smoke, parameterName=Fg_Smoke_Scale}
  n38 = MaterialFunctionCall(
   UVs=Panner(Coordinate=Multiply(A=@n36, B=@n35), Time=NamedRerouteUsage()) {bFractionalPart=true, speedX=-0.05, speedY=0.053}
   Texture Scale=@n37
  ) {materialFunction=&ScaleUVsByCenter}
  n39 = AppendVector(
   A=ScalarParameter() {defaultValue=0.38, group=Fg_Smoke, parameterName=Fg_Smoke_U2}
   B=ScalarParameter() {defaultValue=0.22, group=Fg_Smoke, parameterName=Fg_Smoke_V2}
  )
  n40 = Multiply(A=Time(), B=ScalarParameter() {defaultValue=0.5, group=Fg_Smoke, parameterName=Fg_Smoke_PanSpeed})
  n41 = MaterialFunctionCall(
   UVs=Panner(Coordinate=Multiply(A=@n36, B=@n39), Time=NamedRerouteDeclaration(None=@n40) {name=Fg_Smoke_Time, nodeColor=<rgba 0.1531 1 0.5493 1>}) {bFractionalPart=true, speedX=-0.05, speedY=0.05}
   Texture Scale=@n37
  ) {materialFunction=&ScaleUVsByCenter}
  n42 = Clamp(
   None=Add(
    A=TextureSample(UVs=@n38.UVs) {texture=/Game/Global/Textures/Default/Tileable_Patterns/T_Smoke_Puff_Tile_Up}.RGB
    B=TextureSample(UVs=@n41.UVs) {texture=&T_Smoke_Puff_Tile_Up}.RGB
   )
   Min=ScalarParameter() {defaultValue=-20, group=Fg_Smoke, parameterName=FG_Smoke_ClampMin}
   Max=ScalarParameter() {defaultValue=6.5, group=Fg_Smoke, parameterName=FG_Smoke_ClampMax}
  ) {maxDefault=6.5, minDefault=-20}
  n43 = NamedRerouteUsage()
  n44 = LinearInterpolate(A=Multiply(A=NamedRerouteUsage(), B=@n43), B=@n43) {constAlpha=0.7}
  n45 = Power(Base=Multiply(A=Add(A=@n42, B=@n44)) {constB=1.5}, Exp=ScalarParameter() {defaultValue=0.5, group=Fg_Smoke, parameterName=Fg_Smoke_PowerExp}) {constExponent=3}
  n46 = NamedRerouteUsage()
  NamedRerouteDeclaration(None=Saturate(None=Multiply(A=@n45, B=@n46))) {name=Smoke_Foreground, nodeColor=<rgba 0 0.2384 1 1>}
  NamedRerouteDeclaration(None=@n30.Scaled to Fit) {name=Normal UV Scaled to Fit, nodeColor=<rgba 0 0 0 1>}
  n47 = MaterialFunctionCall(UV=MaterialFunctionCall(UVs=NamedRerouteUsage(), Texture Scale=Constant() {r=10}) {materialFunction=&ScaleUVsByCenter}.UVs) {materialFunction=&MF_UI_RadialGradient}
  n48 = NamedRerouteUsage()
  n49 = Multiply(A=@n48, B=@n15)
  n50 = NamedRerouteDeclaration(None=Multiply(A=Time(), B=ScalarParameter() {defaultValue=1, group=Bg_Smoke, parameterName=Bg_Smoke_TimeMultiply})) {name=Smoke_Bkg_Time, nodeColor=<rgba 0.3255 1 0.0316 1>}
  n51 = Multiply(A=@n48, B=@n16)
  n52 = NamedRerouteUsage()
  n53 = LinearInterpolate(
   A=TextureSample(UVs=Panner(Coordinate=@n49, Time=@n50) {bFractionalPart=true, speedX=0.012, speedY=0.053}) {samplerType=SAMPLERTYPE_LinearGrayscale, texture=/Game/Global/Textures/Default/Tileable_Patterns/T_StormCloudDetail01_M}.G
   B=TextureSample(UVs=Panner(Coordinate=@n51, Time=@n52) {bFractionalPart=true, speedX=-0.016, speedY=0.02}) {texture=/Game/Athena/SafeZone/S11/T_TilingClouds_D}.G
  )
  n54 = Multiply(A=@n53, B=@n53)
  NamedRerouteDeclaration(None=Multiply(A=Multiply(A=@n47.Smooth, B=@n54), B=ScalarParameter() {defaultValue=1, group=Bg_Smoke, parameterName=Bg_Smoke_OpacityMultiply})) {name=Smoke_Background, nodeColor=<rgba 0 0.2384 1 1>}
  n55 = ScalarParameter() {group=Radial, parameterName=Radial_X}
  n56 = ScalarParameter() {defaultValue=0.175, group=Radial, parameterName=Radial_Y}
  n57 = MaterialFunctionCall(
   UVs=Add(A=NamedRerouteUsage(), B=AppendVector(A=@n55, B=@n56))
   Texture Scale=ScalarParameter() {defaultValue=1.6, group=Radial, parameterName=Radial_ScaleMultiplier}
  ) {materialFunction=&ScaleUVsByCenter}
  n58 = SmoothStep(
   Min=ScalarParameter() {defaultValue=-0.05, group=Radial, parameterName=Radial_SmoothStepMin}
   Max=ScalarParameter() {defaultValue=0.95, group=Radial, parameterName=Radial_SmoothStepMax}
   Value=MaterialFunctionCall(
    UV=NamedRerouteDeclaration(None=@n57.UVs) {name=RadialTransform, nodeColor=<rgba 0 0 0 1>}
    GradientStart=ScalarParameter() {defaultValue=0.96, group=Radial, parameterName=Radial_GradStart}
    GradientEnd=ScalarParameter() {group=Radial, parameterName=Radial_GradEnd}
   ) {materialFunction=&MF_UI_RadialGradient}.Smooth
  )
  n59 = CurveAtlasRowParameter(CurveTime=@n58) {
   atlas=/Game/Athena/UI/Events/LinedNotebook/CA_UI_LinedNotebook_BGCurveAtlas
   curve=/Game/Athena/UI/Events/LinedNotebook/C_UI_LinedNotebook_BGCurveGreen
   defaultValue=1
   parameterName=GreenColorCurve
  }
  n60 = CurveAtlasRowParameter(CurveTime=@n58) {
   atlas=&CA_UI_LinedNotebook_BGCurveAtlas
   curve=/Game/Athena/UI/Events/LinedNotebook/C_UI_LinedNotebook_BGCurvePurple
   parameterName=PurpleColorCurve
  }
  NamedRerouteDeclaration(None=StaticSwitchParameter(
    True=MaterialFunctionCall(R=@n59.R, G=@n59.R, B=@n59.R) {materialFunction=&AppendMany}.RGBA
    False=MaterialFunctionCall(R=@n60.R, G=@n60.R, B=@n60.R) {materialFunction=&AppendMany}.RGBA
   ) {defaultValue=true, parameterName=GreenColorScheme}) {name=Radial, nodeColor=<rgba 0 0.2384 1 1>}
  n61 = NamedRerouteUsage()
  NamedRerouteDeclaration(None=LinearInterpolate(A=NamedRerouteUsage(), B=@n61, Alpha=@n61)) {name=Smoke_Merged, nodeColor=<rgba 0 0.2384 1 1>}
  n62 = VectorParameter() {parameterName=SmokeOverlay_GreenLightColor, defaultValue=<rgba 0.0018 0.0156 0.0003 1>}
  n63 = VectorParameter() {parameterName=SmokeOverlay_GreenDarkColor, defaultValue=<rgba 0 0 0 1>}
  n64 = LinearInterpolate(
   A=MaterialFunctionCall(R=@n62.R, G=@n62.R, B=@n62.R) {materialFunction=&AppendMany}.RGBA
   B=MaterialFunctionCall(R=@n63.R, G=@n63.R, B=@n63.R) {materialFunction=&AppendMany}.RGBA
   Alpha=NamedRerouteUsage()
  )
  n65 = VectorParameter() {parameterName=SmokeOverlay_PurpleLightColor, defaultValue=<rgba 0.0212 0.0021 0.0369 1>}
  n66 = VectorParameter() {parameterName=SmokeOverlay_PurpleDarkColor, defaultValue=<rgba 0.0033 0.0009 0.007 1>}
  n67 = NamedRerouteUsage()
  n68 = LinearInterpolate(
   A=MaterialFunctionCall(R=@n65.R, G=@n65.R, B=@n65.R) {materialFunction=&AppendMany}.RGBA
   B=MaterialFunctionCall(R=@n66.R, G=@n66.R, B=@n66.R) {materialFunction=&AppendMany}.RGBA
   Alpha=@n67
  )
  NamedRerouteDeclaration(None=LinearInterpolate(A=NamedRerouteUsage(), B=StaticSwitchParameter(True=@n64, False=@n68) {defaultValue=true, parameterName=GreenColorScheme}, Alpha=@n67)) {name=Merged Background, nodeColor=<rgba 0 0.2384 1 1>}
  NamedRerouteUsage()
  NamedRerouteUsage()
  NamedRerouteUsage()
  NamedRerouteUsage()
 ]
}
